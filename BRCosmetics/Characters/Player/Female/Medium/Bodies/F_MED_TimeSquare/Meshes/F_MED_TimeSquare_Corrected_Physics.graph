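# Physics asset for the F_MED_TimeSquare female body cosmetic, defining secondary-motion bodies for shirt flaps around the hips.
# Capsules cover pelvis, thighs and each flap, with tiny MSTR capsules as flap roots.
# Front-leg roots constrain to the thighs, others to the pelvis, all with free rotation except the locked right side-back root.
# Flap bodies hinge from their roots with limited primary swing, and mostly locked or narrow secondary swing and twist.
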PhysicsAsset /BRCosmetics/Characters/Player/Female/Medium/Bodies/F_MED_TimeSquare/Meshes/F_MED_TimeSquare_Corrected_Physics
bodies {
 dyn_shirt_bk_l_01 [{
   shapeType=Capsule
   length=20.3631
   radius=3.6143
   shapeName=dyn_shirt_bk_l_01_capsule
   center=<13.7618 4.3788 1.2482>
   rotation=<rot 88.8958 1.7916 10.5944>
  }]
 dyn_shirt_bk_l_MSTR [{shapeType=Capsule, length=1, radius=1, shapeName=dyn_shirt_bk_l_MSTR_capsule}]
 dyn_shirt_bk_r_01 [{shapeType=Capsule, length=20.0775, radius=4.6143, shapeName=dyn_shirt_bk_r_01_capsule, center=<15.1877 3.08 0>, rotation=<rot 90 0 8.5782>}]
 dyn_shirt_bk_r_MSTR [{shapeType=Capsule, length=1, radius=1, shapeName=dyn_shirt_bk_r_MSTR_capsule}]
 dyn_shirt_fr_01 [{shapeType=Capsule, length=62.1456, radius=3.7388, shapeName=dyn_shirt_fr_01_capsule, center=<17.8903 0.4634 0.0762>}]
 dyn_shirt_fr_MSTR [{shapeType=Capsule, length=1, radius=1, shapeName=dyn_shirt_fr_MSTR_capsule}]
 dyn_shirt_fr_leg_l_01 [{
   shapeType=Capsule
   length=14.2563
   radius=2.503
   shapeName=dyn_shirt_fr_leg_l_01_capsule
   center=<10.9515 1.4031 0>
   rotation=<rot 90 180 -177.765>
  }]
 dyn_shirt_fr_leg_l_MSTR [{shapeType=Capsule, length=1, radius=1, shapeName=dyn_shirt_fr_leg_l_MSTR_capsule}]
 dyn_shirt_fr_leg_r_01 [{
   shapeType=Capsule
   length=12.8879
   radius=3.003
   shapeName=dyn_shirt_fr_leg_r_01_capsule
   center=<10.9736 0.8234 0>
   rotation=<rot 86.7912 0.0002 0>
  }]
 dyn_shirt_fr_leg_r_MSTR [{shapeType=Capsule, length=1, radius=1, shapeName=dyn_shirt_fr_leg_r_MSTR_capsule}]
 dyn_shirt_side_l_01 [{
   shapeType=Capsule
   length=16.3677
   radius=3.3821
   shapeName=dyn_shirt_side_l_01_capsule
   center=<10.874 3.0285 0>
   rotation=<rot 90 -167.0926 -159.0962>
  }]
 dyn_shirt_side_l_MSTR [{shapeType=Capsule, length=1, radius=1, shapeName=dyn_shirt_side_l_MSTR_capsule}]
 dyn_shirt_side_r_01 [{
   shapeType=Capsule
   length=13.1704
   radius=4.0758
   shapeName=dyn_shirt_side_r_01_capsule
   center=<11.1217 4.9962 0>
   rotation=<rot 90 180 -175.8612>
  }]
 dyn_shirt_side_r_MSTR [{shapeType=Capsule, length=1, radius=1, shapeName=dyn_shirt_side_r_MSTR_capsule}]
 dyn_shirt_sidebk_l_01 [{
   shapeType=Capsule
   length=16.6558
   radius=3.6143
   shapeName=dyn_shirt_sidebk_l_01_capsule
   center=<12.7879 5.5303 0>
   rotation=<rot 90 180 -174.9999>
  }]
 dyn_shirt_sidebk_l_MSTR [{shapeType=Capsule, length=1, radius=1, shapeName=dyn_shirt_sidebk_l_MSTR_capsule}]
 dyn_shirt_sidebk_r_01 [{
   shapeType=Capsule
   length=14.1558
   radius=4.6143
   shapeName=dyn_shirt_sidebk_r_01_capsule
   center=<12.3834 3.9502 0>
   rotation=<rot 90 0 5.6553>
  }]
 dyn_shirt_sidebk_r_MSTR [{shapeType=Capsule, length=1, radius=1, shapeName=dyn_shirt_sidebk_r_MSTR_capsule}]
 pelvis [{shapeType=Capsule, length=11.8347, radius=11.3, shapeName=pelvis_capsule, center=<-16.7008 -0.052 0>}]
 thigh_l [{
   shapeType=Capsule
   length=31.1353
   radius=10.0598
   shapeName=thigh_l_capsule
   center=<-23.3024 0.4221 0.2725>
   rotation=<rot -90 -90 89.9999>
  }]
 thigh_r [{shapeType=Capsule, length=31.6353, radius=9.5598, shapeName=thigh_r_capsule, center=<22.9778 0.4059 0>, rotation=<rot -90 -179.578 179.5779>}]
}
constraints [
 {
  bone1Name=thigh_l
  bone2Name=pelvis
  swing1LimitDegrees=45
  swing1Motion=Free
  swing2LimitDegrees=45
  swing2Motion=Free
  twistLimitDegrees=45
  twistMotion=Free
 }
 {
  bone1Name=thigh_r
  bone2Name=pelvis
  swing1LimitDegrees=45
  swing1Motion=Free
  swing2LimitDegrees=45
  swing2Motion=Free
  twistLimitDegrees=45
  twistMotion=Free
 }
 {
  bone1Name=dyn_shirt_bk_l_MSTR
  bone2Name=pelvis
  swing1LimitDegrees=45
  swing1Motion=Free
  swing2LimitDegrees=45
  swing2Motion=Free
  twistLimitDegrees=45
  twistMotion=Free
 }
 {
  bone1Name=dyn_shirt_bk_l_01
  bone2Name=dyn_shirt_bk_l_MSTR
  swing1LimitDegrees=45
  swing2LimitDegrees=5
  twistLimitDegrees=1
  twistMotion=Locked
 }
 {
  bone1Name=dyn_shirt_bk_r_MSTR
  bone2Name=pelvis
  swing1LimitDegrees=45
  swing1Motion=Free
  swing2LimitDegrees=45
  swing2Motion=Free
  twistLimitDegrees=45
  twistMotion=Free
 }
 {
  bone1Name=dyn_shirt_bk_r_01
  bone2Name=dyn_shirt_bk_r_MSTR
  swing1LimitDegrees=45
  swing2LimitDegrees=5
  twistLimitDegrees=1
  twistMotion=Locked
 }
 {
  bone1Name=dyn_shirt_fr_MSTR
  bone2Name=pelvis
  swing1LimitDegrees=45
  swing1Motion=Free
  swing2LimitDegrees=45
  swing2Motion=Free
  twistLimitDegrees=45
  twistMotion=Free
 }
 {
  bone1Name=dyn_shirt_fr_01
  bone2Name=dyn_shirt_fr_MSTR
  swing1LimitDegrees=60
  swing2LimitDegrees=5
  swing2Motion=Locked
  twistLimitDegrees=40
 }
 {
  bone1Name=dyn_shirt_side_l_MSTR
  bone2Name=pelvis
  swing1LimitDegrees=45
  swing1Motion=Free
  swing2LimitDegrees=45
  swing2Motion=Free
  twistLimitDegrees=45
  twistMotion=Free
 }
 {
  bone1Name=dyn_shirt_side_l_01
  bone2Name=dyn_shirt_side_l_MSTR
  swing1LimitDegrees=30
  swing2LimitDegrees=0
  swing2Motion=Locked
  twistLimitDegrees=0
  twistMotion=Locked
 }
 {
  bone1Name=dyn_shirt_side_r_MSTR
  bone2Name=pelvis
  swing1LimitDegrees=45
  swing1Motion=Free
  swing2LimitDegrees=45
  swing2Motion=Free
  twistLimitDegrees=45
  twistMotion=Free
 }
 {bone1Name=dyn_shirt_side_r_01, bone2Name=dyn_shirt_side_r_MSTR, swing1LimitDegrees=30, swing2LimitDegrees=5, twistLimitDegrees=1}
 {
  bone1Name=dyn_shirt_sidebk_l_MSTR
  bone2Name=pelvis
  swing1LimitDegrees=45
  swing1Motion=Free
  swing2LimitDegrees=45
  swing2Motion=Free
  twistLimitDegrees=45
  twistMotion=Free
 }
 {
  bone1Name=dyn_shirt_sidebk_l_01
  bone2Name=dyn_shirt_sidebk_l_MSTR
  swing1LimitDegrees=45
  swing2LimitDegrees=5
  swing2Motion=Locked
  twistLimitDegrees=1
  twistMotion=Locked
 }
 {
  bone1Name=dyn_shirt_sidebk_r_MSTR
  bone2Name=pelvis
  swing1LimitDegrees=45
  swing1Motion=Locked
  swing2LimitDegrees=45
  swing2Motion=Locked
  twistLimitDegrees=45
  twistMotion=Locked
 }
 {
  bone1Name=dyn_shirt_sidebk_r_01
  bone2Name=dyn_shirt_sidebk_r_MSTR
  swing1LimitDegrees=45
  swing2LimitDegrees=5
  swing2Motion=Locked
  twistLimitDegrees=1
  twistMotion=Locked
 }
 {
  bone1Name=dyn_shirt_fr_leg_l_MSTR
  bone2Name=thigh_l
  swing1LimitDegrees=45
  swing1Motion=Free
  swing2LimitDegrees=45
  swing2Motion=Free
  twistLimitDegrees=45
  twistMotion=Free
 }
 {
  bone1Name=dyn_shirt_fr_leg_l_01
  bone2Name=dyn_shirt_fr_leg_l_MSTR
  swing1LimitDegrees=30
  swing2LimitDegrees=5
  swing2Motion=Locked
  twistLimitDegrees=1
  twistMotion=Locked
 }
 {
  bone1Name=dyn_shirt_fr_leg_r_01
  bone2Name=dyn_shirt_fr_leg_r_MSTR
  swing1LimitDegrees=30
  swing2LimitDegrees=5
  swing2Motion=Locked
  twistLimitDegrees=1
  twistMotion=Locked
 }
 {
  bone1Name=dyn_shirt_fr_leg_r_MSTR
  bone2Name=thigh_r
  swing1LimitDegrees=45
  swing1Motion=Free
  swing2LimitDegrees=45
  swing2Motion=Free
  twistLimitDegrees=45
  twistMotion=Free
 }
]
