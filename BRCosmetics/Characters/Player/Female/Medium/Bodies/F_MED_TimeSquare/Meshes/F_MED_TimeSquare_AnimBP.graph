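# Post-process style AnimBP for the F_MED_TimeSquare female body cosmetic, adding secondary motion to its shirt and shorts.
# The copied parent pose runs through Control Rig and RigidBody, then cached stages add AnimDynamics on shorts and additive bone offsets on shorts and shirt.
# The event graph sets RigidBody gravity with a long-backbling check, computes movement alphas, and feeds calf-velocity WPO and skydive wind to dynamic materials.
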AnimBlueprint /BRCosmetics/Characters/Player/Female/Medium/Bodies/F_MED_TimeSquare/Meshes/F_MED_TimeSquare_AnimBP
target_skeleton=&/F_MED_TimeSquare_Skeleton
graphs [
 AnimGraph {nodes [
   n1 = "Transform (Modify) Bone - Bone: dyn_shirt_fr_leg_l_MSTR"!(ComponentPose, bAlphaBoolEnabled:bool=True, Translation="0.642,1.688,-9.610", Rotation="18.320,-18.873,36.947", Scale="1,1,1") -> (Pose) {
    alphaInputType=Bool
    rotationMode=BMM_Additive
    rotationSpace=BCS_BoneSpace
    translationMode=BMM_Additive
    translationSpace=BCS_BoneSpace
    alphaBoolBlend {blendInTime=0.3, blendOption=ExpInOut, blendOutTime=0.5}
    boneToModify {boneName=dyn_shirt_fr_leg_l_MSTR}
    rotation=<rot 18.32 -18.873 36.947>
    translation=<0.642 1.688 -9.61>
   }
   "Transform (Modify) Bone - Bone: dyn_shirt_side_l_MSTR"!(ComponentPose=@n1, bAlphaBoolEnabled:bool=True, Translation="0,0,0", Rotation="0,0,0", Scale="1,1,1") -> (Pose) {
    alphaInputType=Bool
    rotationMode=BMM_Additive
    rotationSpace=BCS_BoneSpace
    translationMode=BMM_Additive
    translationSpace=BCS_BoneSpace
    alphaBoolBlend {blendInTime=0.3, blendOption=ExpInOut, blendOutTime=0.5}
    boneToModify {boneName=dyn_shirt_side_l_MSTR}
   }
   n2 = "Transform (Modify) Bone - Bone: dyn_shirt_fr_MSTR"!(
    ComponentPose=Local To Component(LocalPose=Use cached pose 'Post-Shorts Setup'())
    bAlphaBoolEnabled:bool=True
    Translation="0,0,0"
    Rotation="0,10,0"
    Scale="1,1,1"
   ) -> (Pose) {
    alphaInputType=Bool
    rotationMode=BMM_Additive
    rotationSpace=BCS_BoneSpace
    translationMode=BMM_Additive
    translationSpace=BCS_BoneSpace
    alphaBoolBlend {blendInTime=0.3, blendOption=ExpInOut, blendOutTime=0.5}
    boneToModify {boneName=dyn_shirt_fr_MSTR}
    rotation {yaw=10}
   }
   n3 = "Transform (Modify) Bone - Bone: dyn_shirt_fr_MSTR"(ComponentPose=@n2, bAlphaBoolEnabled=True, Translation="0,0,0", Rotation="0,10,0", Scale="1,1,1") {
    alphaInputType=Bool
    rotationMode=BMM_Additive
    rotationSpace=BCS_BoneSpace
    translationMode=BMM_Additive
    translationSpace=BCS_BoneSpace
    alphaBoolBlend {blendInTime=0.3, blendOption=ExpInOut, blendOutTime=0.5}
    boneToModify {boneName=dyn_shirt_fr_MSTR}
    rotation {yaw=10}
   }
   n4 = "Transform (Modify) Bone - Bone: dyn_shirt_bk_r_MSTR"(ComponentPose=@n3, Alpha=1, Translation="0,0,0", Rotation="0,5,0", Scale="1,1,1") {
    rotationMode=BMM_Additive
    rotationSpace=BCS_BoneSpace
    translationMode=BMM_Additive
    translationSpace=BCS_BoneSpace
    alphaBoolBlend {blendInTime=0.3, blendOption=ExpInOut, blendOutTime=0.5}
    boneToModify {boneName=dyn_shirt_bk_r_MSTR}
    rotation {yaw=5}
   }
   n5 = "Transform (Modify) Bone - Bone: dyn_shirt_bk_l_MSTR"(ComponentPose=@n4, Alpha=1, Translation="0,0,0", Rotation="0,5,0", Scale="1,1,1") {
    rotationMode=BMM_Additive
    rotationSpace=BCS_BoneSpace
    translationMode=BMM_Additive
    translationSpace=BCS_BoneSpace
    alphaBoolBlend {blendInTime=0.3, blendOption=ExpInOut, blendOutTime=0.5}
    boneToModify {boneName=dyn_shirt_bk_l_MSTR}
    rotation {yaw=5}
   }
   n6 = "Transform (Modify) Bone - Bone: dyn_shirt_side_l_MSTR"!(ComponentPose=@n5, Alpha:float=1, Translation="0,0,0", Rotation="-30,25,0", Scale="1,1,1") -> (Pose) {
    rotationMode=BMM_Additive
    rotationSpace=BCS_BoneSpace
    translationMode=BMM_Additive
    translationSpace=BCS_BoneSpace
    alphaBoolBlend {blendInTime=0.3, blendOption=ExpInOut, blendOutTime=0.5}
    boneToModify {boneName=dyn_shirt_side_l_MSTR}
    rotation {pitch=-30, yaw=25}
   }
   n7 = "Transform (Modify) Bone - Bone: dyn_shirt_side_l_MSTR"(ComponentPose=@n6, Alpha=1, Translation="0,0,0", Rotation="-30,25,0", Scale="1,1,1") {
    rotationMode=BMM_Additive
    rotationSpace=BCS_BoneSpace
    translationMode=BMM_Additive
    translationSpace=BCS_BoneSpace
    alphaBoolBlend {blendInTime=0.3, blendOption=ExpInOut, blendOutTime=0.5}
    boneToModify {boneName=dyn_shirt_side_l_MSTR}
    rotation {pitch=-30, yaw=25}
   }
   n8 = "Transform (Modify) Bone - Bone: dyn_shirt_bk_l_MSTR"!(ComponentPose=@n7, bAlphaBoolEnabled:bool=True, Translation="0,0,0", Rotation="0,5,0", Scale="1,1,1") -> (Pose) {
    alphaInputType=Bool
    rotationMode=BMM_Additive
    rotationSpace=BCS_BoneSpace
    translationMode=BMM_Additive
    translationSpace=BCS_BoneSpace
    alphaBoolBlend {blendInTime=0.3, blendOption=ExpInOut, blendOutTime=0.5}
    boneToModify {boneName=dyn_shirt_bk_l_MSTR}
    rotation {yaw=5}
   }
   n9 = "Transform (Modify) Bone - Bone: dyn_shirt_bk_r_MSTR"!(ComponentPose=@n8, bAlphaBoolEnabled:bool=True, Translation="0,0,0", Rotation="0,5,0", Scale="1,1,1") -> (Pose) {
    alphaInputType=Bool
    rotationMode=BMM_Additive
    rotationSpace=BCS_BoneSpace
    translationMode=BMM_Additive
    translationSpace=BCS_BoneSpace
    alphaBoolBlend {blendInTime=0.3, blendOption=ExpInOut, blendOutTime=0.5}
    boneToModify {boneName=dyn_shirt_bk_r_MSTR}
    rotation {yaw=5}
   }
   n10 = "Transform (Modify) Bone - Bone: dyn_shirt_sidebk_r_MSTR"!(ComponentPose=@n9, bAlphaBoolEnabled:bool=True, Translation="0,0,2", Rotation="0,2,0", Scale="1,1,1") -> (Pose) {
    alphaInputType=Bool
    rotationMode=BMM_Additive
    rotationSpace=BCS_BoneSpace
    translationMode=BMM_Additive
    translationSpace=BCS_BoneSpace
    alphaBoolBlend {blendInTime=0.3, blendOption=ExpInOut, blendOutTime=0.5}
    boneToModify {boneName=dyn_shirt_sidebk_r_MSTR}
    rotation {yaw=2}
    translation {z=2}
   }
   n11 = "Transform (Modify) Bone - Bone: dyn_shirt_sidebk_l_MSTR"!(ComponentPose=@n10, bAlphaBoolEnabled:bool=True, Translation="0,0,0", Rotation="0,6,0", Scale="1,1,1") -> (Pose) {
    alphaInputType=Bool
    rotationMode=BMM_Additive
    rotationSpace=BCS_BoneSpace
    translationMode=BMM_Additive
    translationSpace=BCS_BoneSpace
    alphaBoolBlend {blendInTime=0.3, blendOption=ExpInOut, blendOutTime=0.5}
    boneToModify {boneName=dyn_shirt_sidebk_l_MSTR}
    rotation {yaw=6}
   }
   n12 = "Transform (Modify) Bone - Bone: dyn_shirt_sidebk_r_MSTR"(ComponentPose=@n11, bAlphaBoolEnabled=True, Translation="0,0,0", Rotation="0,10,0", Scale="1,1,1") {
    alphaInputType=Bool
    rotationMode=BMM_Additive
    rotationSpace=BCS_BoneSpace
    translationMode=BMM_Additive
    translationSpace=BCS_BoneSpace
    alphaBoolBlend {blendInTime=0.3, blendOption=ExpInOut, blendOutTime=0.5}
    boneToModify {boneName=dyn_shirt_sidebk_r_MSTR}
    rotation {yaw=10}
   }
   n13 = "Transform (Modify) Bone - Bone: dyn_shirt_sidebk_l_MSTR"(ComponentPose=@n12, bAlphaBoolEnabled=True, Translation="0,0,0", Rotation="0,10,0", Scale="1,1,1") {
    alphaInputType=Bool
    rotationMode=BMM_Additive
    rotationSpace=BCS_BoneSpace
    translationMode=BMM_Additive
    translationSpace=BCS_BoneSpace
    alphaBoolBlend {blendInTime=0.3, blendOption=ExpInOut, blendOutTime=0.5}
    boneToModify {boneName=dyn_shirt_sidebk_l_MSTR}
    rotation {yaw=10}
   }
   Output Pose(Result=Component To Local(ComponentPose=@n13)) {layerGroup=None}
   n14 = Anim Dynamics - Bone: deform_shorts_r_MSTR!(ComponentPose=Local To Component(LocalPose=Use cached pose 'Post RB'()), Alpha=0.4, GravityOverride:Vector="0,0,0", ExternalForce="0,0,0") -> (Pose) {
    angularDampingOverride=0.9
    angularSpringConstant=50
    bAngularSpring=true
    bOverrideAngularDamping=true
    bUseGravityOverride=true
    simulationSpace=BoneRelative
    alphaBoolBlend {blendOption=Linear}
    boundBone {boneName=deform_shorts_r_MSTR}
    componentAppliedLinearAccClamp=<1000 1000 1000>
    componentLinearAccScale=<1 1 1>
    componentLinearVelScale=<1 1 1>
    gravityOverride {z=-980}
    physicsBodyDefinitions [{
      boundBone {boneName=deform_shorts_r_MSTR}
      boxExtents=<15 15 15>
      constraintSetup {angularLimitsMax=<5 8 15>, angularLimitsMin=<-5 -8 -5>, angularTarget=<0 0 0>}
      localJointOffset=<2 0 0>
     }]
    relativeSpaceBone {boneName=thigh_r}
   }
   n15 = Anim Dynamics - Bone: deform_shorts_l_MSTR!(ComponentPose=@n14, Alpha=0.4, GravityOverride:Vector="0,0,0", ExternalForce="0,0,0") -> (Pose) {
    angularDampingOverride=0.9
    angularSpringConstant=50
    bAngularSpring=true
    bOverrideAngularDamping=true
    bUseGravityOverride=true
    simulationSpace=BoneRelative
    alphaBoolBlend {blendOption=Linear}
    boundBone {boneName=deform_shorts_l_MSTR}
    componentAppliedLinearAccClamp=<1000 1000 1000>
    componentLinearAccScale=<1 1 1>
    componentLinearVelScale=<1 1 1>
    gravityOverride {z=-980}
    physicsBodyDefinitions [{
      boundBone {boneName=deform_shorts_l_MSTR}
      boxExtents=<15 15 15>
      constraintSetup {angularLimitsMax=<15 8 8>, angularLimitsMin=<-15 -8 -2>, angularTarget=<0 0 0>}
      localJointOffset=<-10 0 0>
     }]
    relativeSpaceBone {boneName=thigh_l}
   }
   n16 = "Transform (Modify) Bone - Bone: deform_shorts_l_MSTR"!(ComponentPose=@n15, bAlphaBoolEnabled:bool=True, Translation="0.273,-1.527,-1.852", Rotation="-0.012,10.000,0.016", Scale="1,1,1") -> (Pose) {
    alphaInputType=Bool
    rotationMode=BMM_Additive
    rotationSpace=BCS_BoneSpace
    translationMode=BMM_Additive
    translationSpace=BCS_BoneSpace
    alphaBoolBlend {blendInTime=0.3, blendOption=ExpInOut, blendOutTime=0.5}
    boneToModify {boneName=deform_shorts_l_MSTR}
    rotation=<rot -0.012 9.9999 0.016>
    translation=<0.273 -1.527 -1.852>
   }
   n17 = "Transform (Modify) Bone - Bone: deform_shorts_r_MSTR"!(ComponentPose=@n16, bAlphaBoolEnabled:bool=True, Translation="0,0,0", Rotation="-0.003,5.000,-0.002", Scale="1,1,1") -> (Pose) {
    alphaInputType=Bool
    rotationMode=BMM_Additive
    rotationSpace=BCS_BoneSpace
    translationMode=BMM_Additive
    translationSpace=BCS_BoneSpace
    alphaBoolBlend {blendInTime=0.3, blendOption=ExpInOut, blendOutTime=0.5}
    boneToModify {boneName=deform_shorts_r_MSTR}
    rotation=<rot -0.003 4.9999 -0.002>
   }
   n18 = "Transform (Modify) Bone - Bone: deform_shorts_back_l"!(ComponentPose=@n17, bAlphaBoolEnabled:bool=True, Translation="0,0,0", Rotation="0,30,0", Scale="1,1,1") -> (Pose) {
    alphaInputType=Bool
    rotationMode=BMM_Additive
    rotationSpace=BCS_BoneSpace
    translationMode=BMM_Additive
    translationSpace=BCS_BoneSpace
    alphaBoolBlend {blendInTime=0.3, blendOption=ExpInOut, blendOutTime=0.5}
    boneToModify {boneName=deform_shorts_back_l}
    rotation {yaw=30}
   }
   n19 = "Transform (Modify) Bone - Bone: deform_shorts_back_r"!(ComponentPose=@n18, bAlphaBoolEnabled:bool=True, Translation="0,0,0", Rotation="0,20,0", Scale="1,1,1") -> (Pose) {
    alphaInputType=Bool
    rotationMode=BMM_Additive
    rotationSpace=BCS_BoneSpace
    translationMode=BMM_Additive
    translationSpace=BCS_BoneSpace
    alphaBoolBlend {blendInTime=0.3, blendOption=ExpInOut, blendOutTime=0.5}
    boneToModify {boneName=deform_shorts_back_r}
    rotation {yaw=20}
   }
   Save cached pose 'Post-Shorts Setup'(Pose=Component To Local(ComponentPose=@n19)) {cacheName=Post-Shorts Setup}
   n20 = Control Rig!(Source=Copy Pose From Mesh() {bUseAttachedParent=true}, Alpha, Is in Frontend:bool=Get bIsInFrontEnd()) -> (Pose) {
    alphaBoolBlend {blendOption=Linear}
    controlRigAssetReference {blueprintRigClass=&/F_MED_TimeSquare_CtrlRig.F_MED_TimeSquare_CtrlRig_C}
   }
   n21 = RigidBody!(
    ComponentPose=Local To Component(LocalPose=@n20)
    Alpha
    OverrideWorldGravity:Vector=Get Override World Gravity() -> (Override World Gravity:Vector)
    ExternalForce
   ) -> (Pose) {
    bClampLinearTranslationLimitToRefPose=true
    bOverrideWorldGravity=true
    overridePhysicsAsset=&/F_MED_TimeSquare_Corrected_Physics
    simulationSpace=BaseBoneSpace
    alphaBoolBlend {blendOption=Linear}
    baseBoneRef {boneName=spine_01}
    componentAppliedLinearAccClamp=<2000 2000 2000>
    componentLinearAccScale=<0.3 0.3 0.3>
    componentLinearVelScale=<0.3 0.3 0.3>
    overrideWorldGravity {z=-980}
   }
   Save cached pose 'Post RB'(Pose=Component To Local(ComponentPose=@n21)) {cacheName=Post RB}
  ]}
 EventGraph {nodes [
   Event OnCharacterCustomizationLoaded()
   n1 = GetPartSkeletalMeshComponent()
   > n2 = Set SkelMesh(SkelMesh=@n1)
   > n3 = Get Dynamic Material Instance(SkeletalMeshComponent=@n2)
   > Set Mat Instance Dynamic(Mat Instance Dynamic=@n3)
   > n4 = GetOrCreateDynamicOverlayMaterial(MeshComponent=@n1, MaterialName=None)
   > Set Overlay Mat Dynamic(Overlay Mat Dynamic:Material Instance Dynamic Object Reference=@n4) -> (Output_Get:Material Instance Dynamic Object Reference)
   Event BlueprintUpdateAnimation()
   > Set Delta Time X(Delta Time X=0)
   > n5 = Sequence!(execute) -> (then_0, then_1, then_2:Exec, then_3:Exec)
   > n6 = Gravity Override State Function DEPRECATED(
    CharacterPartAnimInstance=Self-Reference()
    GameTimeInSeconds=Get Game Time in Seconds() -> (Game Time in Seconds:double)
    GravityOverrideParams=Get Gravity Override Params()
   )
   > Set Override World Gravity(Override World Gravity:Vector=SelectVector(A="0,0,-980", B=@n6, bPickA=CheckForBackbling().bIsBackblingLong)) -> (Output_Get:Vector)
   Is Valid(exec=@n5.then_1, InputObject=Get PlayerPawn() -> (PlayerPawn:Pawn Object Reference))
   > Set WPO Globals Function(
    Material=Get Mat Instance Dynamic()
    Overlay Material=Get Overlay Mat Dynamic() -> (Overlay Mat Dynamic:Material Instance Dynamic Object Reference)
    Skeletal Mesh=Get SkelMesh()
    "Local Noise Transform  Location"="Get Local Noise Transform  Location"() -> ("Local Noise Transform  Location":Vector)
    Local Noise Offset=Get Local Noise Offset()
   )
   > Set WPO Regional Function!(
    execute
    Material:Material Instance Dynamic Object Reference=Get Mat Instance Dynamic()
    Overlay Material:Material Instance Dynamic Object Reference=Get Overlay Mat Dynamic()
    Skeletal Mesh:Skeletal Mesh Component Object Reference=Get SkelMesh()
    Velocity Transform:Name=calf_r
    Velocity Interp Speed:double=Get Velocity Interp Speed()
    Collision Transform:Name=pelvis
    Collision Transform Axis:EAxis Enum
    Invert Collision Transform Axis:bool=false
    Delta Time X:double=Get Delta Time X()
    Previous Velocity Transform Location:"Vector (by ref)"=Get R1_Previous Velocity Transform Location() -> (R1_Previous Velocity Transform Location:Vector)
    Previous Velocity Transform Velocity:"Vector (by ref)"=Get R1_Previous Velocity Transform Velocity() -> (R1_Previous Velocity Transform Velocity:Vector)
    Region:int=1
   ) -> (then)
   > Set WPO Regional Function(
    Material=Get Mat Instance Dynamic()
    Overlay Material=Get Overlay Mat Dynamic()
    Skeletal Mesh=Get SkelMesh()
    Velocity Transform=calf_l
    Velocity Interp Speed=Get Velocity Interp Speed()
    Collision Transform=pelvis
    Invert Collision Transform Axis=false
    Delta Time X=Get Delta Time X()
    Previous Velocity Transform Location=Get R2_Previous Velocity Transform Location() -> (R2_Previous Velocity Transform Location:Vector)
    Previous Velocity Transform Velocity=Get R2_Previous Velocity Transform Velocity() -> (R2_Previous Velocity Transform Velocity:Vector)
    Region=2
   )
   > n7 = Branch(Condition=Get bIsSkydiveFloating())
   > n8 = Set WPO Pho Wind Intensity(WPO Pho Wind Intensity=1)
   SetScalarParameterValue(execute=[@n9.then, @n8.then], self=Get Mat Instance Dynamic(), ParameterName=ClothSim_WindStrength, Value=Get WPO Pho Wind Intensity())
   n9 = Set WPO Pho Wind Intensity(execute=@n7.else, WPO Pho Wind Intensity=0)
   Set is in Air(execute=@n5.then_2, is in Air:bool=OR Boolean(A=Get bIsJumping(), B=Get bIsFalling())) -> (Output_Get:bool)
   n10 = TryGetPawnOwner()
   n11 = GetVelocity!(self:Actor Object Reference=@n10) -> (ReturnValue:Vector)
   n12 = "Absolute (Float)"(A=Dot Product!(A:Vector=@n11, B:Vector=GetActorRightVector!(self:Actor Object Reference=@n10) -> (ReturnValue:Vector)) -> (ReturnValue:double))
   Set Crouching Left-Right Movement Alpha(
    execute=@n5.then_3
    Crouching Left-Right Movement Alpha:double=float * float(A=MapRangeClamped(Value=@n12, InRangeA=0, InRangeB=230), B:bool=Get bIsCrouching()) {is=title K2Node_PromotableOperator}
   ) -> (Output_Get:double)
   n13 = GetActorForwardVector!(self:Actor Object Reference=@n10) -> (ReturnValue:Vector)
   n14 = MapRangeClamped(Value="Absolute (Float)"(A=Dot Product(A=@n11, B=@n13)), InRangeA=0, InRangeB=230)
   > Set Crouching Forward-backward Movement Alpha(Crouching Forward-backward Movement Alpha:double=float * float(A=@n14, B=Get bIsCrouching())) -> (Output_Get:double)
   > Set Forward-backward Movement Alpha(Forward-backward Movement Alpha:double=@n14) -> (Output_Get:double)
   Event BlueprintInitializeAnimation()
   > Set SkelMesh(SkelMesh=GetOwnerSkeletalMeshComponent())
   n15 = TryGetPawnOwner()
   > Is Valid(InputObject=@n15)
   > n16 = Cast To FortPlayerPawnAthena(Object=@n15)
   > Set PlayerPawn(PlayerPawn:Pawn Object Reference=@n16) -> (Output_Get:Pawn Object Reference)
  ]}
]
variables [
 Crouching Forward-backward Movement Alpha {type=number}
 Crouching Left-Right Movement Alpha {type=number}
 Delta Time X {type=number}
 Forward-backward Movement Alpha {type=number}
 Game Time in Seconds {type=number}
 Gravity Override Params {type=GravityOverrideParamsStruct, default {skydive_FloatingDown {gravityOverride=<0 -700 -980>}}}
 is in Air {type=boolean}
 Local Noise Offset {type=Vector}
 "Local Noise Transform  Location" {type=Vector}
 Mat Instance Dynamic {type=MaterialInstanceDynamic}
 Overlay Mat Dynamic {type=MaterialInstanceDynamic}
 Override World Gravity {type=Vector}
 PlayerPawn {type=Pawn, default=null}
 R1_Previous Velocity Transform Location {type=Vector}
 R1_Previous Velocity Transform Velocity {type=Vector}
 R2_Previous Velocity Transform Location {type=Vector}
 R2_Previous Velocity Transform Velocity {type=Vector}
 R3_Previous Velocity Transform Location {type=Vector}
 R3_Previous Velocity Transform Velocity {type=Vector}
 SkelMesh {type=SkeletalMeshComponent}
 Velocity Interp Speed {type=number}
 WPO Pho Wind Intensity {type=number}
]
