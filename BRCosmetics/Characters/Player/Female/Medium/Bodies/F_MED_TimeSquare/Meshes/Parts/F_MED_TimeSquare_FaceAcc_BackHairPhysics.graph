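# Physics asset for the back hair on the female medium TimeSquare character's face accessory part, defining capsule bodies for hair physics.
# Three four-bone chains (centre, left, right) link each segment to the previous one with limited swing and twist.
# Cross constraints tie both side tips to the centre tip with free angular motion.
# Capsules on the neck, upper spine, clavicles and right upper arm (no left) provide body collision shapes.
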PhysicsAsset /BRCosmetics/Characters/Player/Female/Medium/Bodies/F_MED_TimeSquare/Meshes/Parts/F_MED_TimeSquare_FaceAcc_BackHairPhysics
bodies {
 clavicle_l [{
   shapeType=Capsule
   length=8.5
   radius=9.7
   shapeName=clavicle_l_capsule
   center=<2.6242 1.1153 -3.1999>
   rotation=<rot 86.1897 6.7318 -0.0027>
  }]
 clavicle_r [{
   shapeType=Capsule
   length=7.5
   radius=9.7
   shapeName=clavicle_r_capsule
   center=<-3.0882 -1.1367 3.0492>
   rotation=<rot 84.2671 6.423 4.3631>
  }]
 dyn_hair_bk_01 [{shapeType=Capsule, length=2, radius=1, shapeName=dyn_hair_bk_01_capsule, center=<2.5552 0 0>}]
 dyn_hair_bk_02 [{shapeType=Capsule, length=2.5, radius=6, shapeName=dyn_hair_bk_02_capsule, center=<4.4506 -4.2203 0>}]
 dyn_hair_bk_03 [{shapeType=Capsule, length=5, radius=8, shapeName=dyn_hair_bk_03_capsule, center=<6.7859 -5.491 0>}]
 dyn_hair_bk_04 [{shapeType=Capsule, length=4.5, radius=8.5, shapeName=dyn_hair_bk_04_capsule, center=<8.365 -6.62 0>}]
 dyn_hair_bk_l_01 [{shapeType=Capsule, length=2, radius=1, shapeName=dyn_hair_bk_l_01_capsule, center=<2.5552 0 0>}]
 dyn_hair_bk_l_02 [{shapeType=Capsule, length=3, radius=4, shapeName=dyn_hair_bk_l_02_capsule, center=<4 -1.5771 0>}]
 dyn_hair_bk_l_03 [{shapeType=Capsule, length=4, radius=5, shapeName=dyn_hair_bk_l_03_capsule, center=<0.9 -1.35 0>}]
 dyn_hair_bk_l_04 [{shapeType=Capsule, length=9, radius=6, shapeName=dyn_hair_bk_l_04_capsule, center=<6.9 -4.6 -2>, rotation=<rot 56.1761 16.9442 -1.773>}]
 dyn_hair_bk_r_01 [{shapeType=Capsule, length=2, radius=1, shapeName=dyn_hair_bk_r_01_capsule, center=<2.5552 0 0>}]
 dyn_hair_bk_r_02 [{shapeType=Capsule, length=3, radius=4, shapeName=dyn_hair_bk_r_02_capsule, center=<3 -1.6538 0>}]
 dyn_hair_bk_r_03 [{
   shapeType=Capsule
   length=4.5
   radius=5
   shapeName=dyn_hair_bk_r_03_capsule
   center=<4.2445 -2.3 -1.2>
   rotation=<rot 81.9978 -18.9924 -6.2714>
  }]
 dyn_hair_bk_r_04 [{shapeType=Capsule, length=9, radius=5.5, shapeName=dyn_hair_bk_r_04_capsule, center=<7 -3.45 4.3>, rotation=<rot 61.0001 -170.9996 -1.6761>}]
 neck_01 [{shapeType=Capsule, length=13.5, radius=6.5, shapeName=neck_01_capsule, center=<1.8124 -0.0105 0>, rotation=<rot 90 180 -174.0069>}]
 spine_04 [{shapeType=Capsule, length=6, radius=15.5, shapeName=spine_04_capsule, center=<-2.4442 4.8309 0>}]
 spine_05 [{shapeType=Capsule, length=7, radius=14.5, shapeName=spine_05_capsule, center=<-2.3204 4.7698 0>}]
 upperarm_r [{shapeType=Capsule, length=24.5, radius=5.5, shapeName=upperarm_r_capsule, center=<-13.5531 0 0>, rotation=<rot 90 0 0>}]
}
constraints [
 {bone1Name=dyn_hair_bk_02, bone2Name=dyn_hair_bk_01, swing1LimitDegrees=40, swing2LimitDegrees=5, twistLimitDegrees=3}
 {bone1Name=dyn_hair_bk_03, bone2Name=dyn_hair_bk_02, swing1LimitDegrees=35, swing2LimitDegrees=10, twistLimitDegrees=10}
 {bone1Name=dyn_hair_bk_04, bone2Name=dyn_hair_bk_03, swing1LimitDegrees=30, swing2LimitDegrees=15, twistLimitDegrees=10}
 {bone1Name=dyn_hair_bk_l_04, bone2Name=dyn_hair_bk_l_03, swing1LimitDegrees=30, swing2LimitDegrees=15, twistLimitDegrees=2}
 {bone1Name=dyn_hair_bk_l_03, bone2Name=dyn_hair_bk_l_02, swing1LimitDegrees=30, swing2LimitDegrees=10, twistLimitDegrees=2}
 {bone1Name=dyn_hair_bk_l_02, bone2Name=dyn_hair_bk_l_01, swing1LimitDegrees=40, swing2LimitDegrees=5, twistLimitDegrees=10}
 {bone1Name=dyn_hair_bk_r_04, bone2Name=dyn_hair_bk_r_03, swing1LimitDegrees=30, swing2LimitDegrees=15, twistLimitDegrees=10}
 {bone1Name=dyn_hair_bk_r_03, bone2Name=dyn_hair_bk_r_02, swing1LimitDegrees=40, swing2LimitDegrees=10, twistLimitDegrees=2}
 {bone1Name=dyn_hair_bk_r_02, bone2Name=dyn_hair_bk_r_01, swing1LimitDegrees=40, swing2LimitDegrees=1, twistLimitDegrees=10}
 {
  bone1Name=dyn_hair_bk_r_04
  bone2Name=dyn_hair_bk_04
  swing1LimitDegrees=45
  swing1Motion=Free
  swing2LimitDegrees=45
  swing2Motion=Free
  twistLimitDegrees=45
  twistMotion=Free
 }
 {
  bone1Name=dyn_hair_bk_l_04
  bone2Name=dyn_hair_bk_04
  swing1LimitDegrees=45
  swing1Motion=Free
  swing2LimitDegrees=45
  swing2Motion=Free
  twistLimitDegrees=45
  twistMotion=Free
 }
]
